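; Cluster bomb mini-grenade projectile for the ProtoCannon that explodes, damaging targets and launching players and vehicles.
; It detonates when stopped after a minimum delay, at a maximum timer, or on EndPlay if not yet exploded.
; On authority, an async targeting preset feeds team- and rarity-based damage effects and radial-plus-vertical impulses, granting launched pawns fall-damage immunity until landing.
; Cosmetic clients get water or ground explosion effects, audio and a sound indicator.
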(event EventBeginPlay
  ;[0.03429 0.3069 0.7899] Start Min and Max Explosion Timers and Spawn Cluster Grenade GCN
  (Utilities|Time|SetTimerbyEvent (EventDispatchers|CreateEvent :!event Custom|MinExplosionDelay) (ScalableFloat|GetValueatLevel (Variables|ProtoCannon|GetHF_MinExplosionDelay)))
  (Utilities|Time|SetTimerbyEvent (EventDispatchers|CreateEvent :!event Custom|MaxExplosionDelay) (ScalableFloat|GetValueatLevel (Variables|ProtoCannon|GetHF_MaxExplosionDelay)))
  (GameplayCue|ExecuteGameplayCueOnActor(Burst) self (Variables|Default|GetGC_MiniGrenadeSpawned) (Ability|GameplayCue|MakeGameplayCueParameters 0.0)))

(event Custom|MinExplosionDelay
  ;[0.03429 0.3069 0.7899] Min Explosion Timer
  (Variables|/Script/Engine.ProjectileMovementComponent|SetMaxSpeed (ScalableFloat|GetValueatLevel (Variables|ProtoCannon|GetHF_ClusterMaxSpeed)) (Variables|Movement|GetProjectileMovementComponent))
  (Variables|Default|SetbMinExplosionDelay true)
  (if (Variables|Default|GetbHasStopped)
    (CallFunction|Explode)))

(event Custom|MaxExplosionDelay
  ;[0.03429 0.3069 0.7899] Max Explosion Timer
  (CallFunction|Explode))

(event Custom|Explode
  (bind _returnvalue (Transformation|GetActorLocation))
  (bind _fx_explosionscale (Variables|Default|GetFX_ExplosionScale))
  (Variables|Default|SetHasExploded true)
  (CallFunction|ProcExplosion)
  (GameplayCues|RemoveLoopingGameplayCue (Variables|Default|GetGC_MiniGrenadeTrail))
  (Networking|CanExecuteCosmeticEvents
    (:True
      (Audio|PlaySoundatLocation (Variables|Default|GetShockwave%20Sound%20Cue) (Transformation|GetActorLocation))
      (Fortnite|ShowSoundIndicatorLocal 0 (Transformation|GetActorLocation) 5e+03 "Gunshot")
      (if (or (Utilities|Enum|Equal(Enum) (Variables|Default|GetSurfaceType) "SurfaceType8") (Water|IsinWater (Variables|Water|GetWaterInteractionComponent)))
        (bind _returnvalue_2 (Effects|Components|ParticleSystem|SpawnFXAtLocation (Variables|Default|GetExplosion%20Asset_Water) _returnvalue))
        (Effects|Components|ParticleSystem|SetFloatParameter _returnvalue_2 "Global Scale" _fx_explosionscale)
        (Audio|PlaySoundatLocation (Variables|Default|GetExplosionAudioWater) _returnvalue)
        (Audio|PlaySoundatLocation (Variables|Default|GetExplosionAudioWaterSplash) _returnvalue)
        (else
          (bind _returnvalue_3 (Effects|Components|ParticleSystem|SpawnFXAtLocation (Variables|Default|GetExplosion%20Asset) _returnvalue))
          (Effects|Components|ParticleSystem|SetFloatParameter _returnvalue_3 "Global Scale" _fx_explosionscale)
          (Audio|PlaySoundatLocation (Variables|Default|GetExplosionAudio) _returnvalue)
          (Effects|Components|ParticleSystem|SetBoolParameter _returnvalue_3 "SpawnDecal" (not (Variables|Default|GetbMainGrenadeStuckToMovingPlatform)))
          (Effects|Components|ParticleSystem|SetVectorParameter _returnvalue_3 "DecalRot" (Variables|Default|GetHitNormal)))))
    (:False
      _))
  (Utilities|FlowControl|Delay 0.2)
  (Actor|DestroyActor))

(event EventEndPlay
  (params (EndPlayReason :type EEndPlayReason%20Enum :default "Destroyed"))
  (bind _returnvalue (Transformation|GetActorLocation))
  (bind _fx_explosionscale (Variables|Default|GetFX_ExplosionScale))
  (|CallParentFunction EndPlayReason)
  (GameplayCues|RemoveLoopingGameplayCue (Variables|Default|GetGC_MiniGrenadeTrail))
  (if (not (Variables|Default|GetHasExploded))
    (Variables|Default|SetHasExploded true)
    (CallFunction|ProcExplosion)
    (GameplayCues|RemoveLoopingGameplayCue (Variables|Default|GetGC_MiniGrenadeTrail))
    (Networking|CanExecuteCosmeticEvents
      (:True
        (Audio|PlaySoundatLocation (Variables|Default|GetShockwave%20Sound%20Cue) (Transformation|GetActorLocation))
        (Fortnite|ShowSoundIndicatorLocal 0 (Transformation|GetActorLocation) 5e+03 "Gunshot")
        (if (or (Utilities|Enum|Equal(Enum) (Variables|Default|GetSurfaceType) "SurfaceType8") (Water|IsinWater (Variables|Water|GetWaterInteractionComponent)))
          (bind _returnvalue_2 (Effects|Components|ParticleSystem|SpawnFXAtLocation (Variables|Default|GetExplosion%20Asset_Water) _returnvalue))
          (Effects|Components|ParticleSystem|SetFloatParameter _returnvalue_2 "Global Scale" _fx_explosionscale)
          (Audio|PlaySoundatLocation (Variables|Default|GetExplosionAudioWater) _returnvalue)
          (Audio|PlaySoundatLocation (Variables|Default|GetExplosionAudioWaterSplash) _returnvalue)
          (else
            (bind _returnvalue_3 (Effects|Components|ParticleSystem|SpawnFXAtLocation (Variables|Default|GetExplosion%20Asset) _returnvalue))
            (Effects|Components|ParticleSystem|SetFloatParameter _returnvalue_3 "Global Scale" _fx_explosionscale)
            (Audio|PlaySoundatLocation (Variables|Default|GetExplosionAudio) _returnvalue)
            (Effects|Components|ParticleSystem|SetBoolParameter _returnvalue_3 "SpawnDecal" (not (Variables|Default|GetbMainGrenadeStuckToMovingPlatform)))
            (Effects|Components|ParticleSystem|SetVectorParameter _returnvalue_3 "DecalRot" (Variables|Default|GetHitNormal)))))
      (:False
        _))
    (Utilities|FlowControl|Delay 0.2)
    (Actor|DestroyActor)))

(event EventOnStop
  (params (Hit :type Hit%20Result%20Structure%20%28by%20ref%29))
  (bind (_bblockinghit _binitialoverlap _time _distance _location _impactpoint _normal _impactnormal) (Collision|BreakHitResult Hit))
  (Variables|Default|SetHitNormal _impactnormal)
  (Variables|Default|SetbHasStopped true)
  (if (Variables|Default|GetbMinExplosionDelay)
    (CallFunction|Explode)))

(event Custom|ProcExplosion
  (Networking|SwitchHasAuthority
    (:Authority
      (bind (_asynctaskproxy _targetinghandle) (Class|AsyncActionPerformTargeting|PerformTargetingAsyncAction 0 "/PotatoCoaster/Gameplay/ProtoCannon_ClusterBomb/Impulse/TP_ProtoCannon_Cluster_Explosion.TP_ProtoCannon_Cluster_Explosion" true)
        (:Targeted
          (for _array_element (TargetingSystem|TargetingResults|GetTargetingResults (GameInstanceSubsystems|GetTargetingSubsystem) _targetinghandle)
            (bind (_array_element_bblockinghit _array_element_binitialoverlap _array_element_time _array_element_distance _array_element_location _array_element_impactpoint _array_element_normal _array_element_impactnormal _array_element_physmat _array_element_hitactor) (Collision|BreakHitResult _array_element))
            (CallFunction|ApplyDamage _array_element_hitactor)
            (bind _asfort_player_pawn (Utilities|Casting|CastToFortPlayerPawn _array_element_hitactor)
              (:then
                (Utilities|ShouldApplyImpulse _asfort_player_pawn
                  (:True
                    (Character|LaunchCharacter _asfort_player_pawn (CallFunction|CalculateImpulse _asfort_player_pawn (ScalableFloat|GetValueatLevel (Variables|ProtoCannon|GetHF_ImpulseMultiplier)) (ScalableFloat|GetValueatLevel (Variables|ProtoCannon|GetHF_AdditionalVerticalImpulse))) true true)
                    (Ability|Advanced|GrantandActivateAbilityonSelf (Ability|GetFortAbilitySystemComponent _asfort_player_pawn) "/PotatoCoaster/Gameplay/GAS/GA_PotatoCoaster_TillLandFallDamageImmunity_ProtoCannon.GA_PotatoCoaster_TillLandFallDamageImmunity_ProtoCannon_C"))))
              (:CastFailed
                (bind _asfort_athena_vehicle (Utilities|Casting|CastToFortAthenaVehicle _array_element_hitactor))
                (Utilities|ShouldApplyImpulse _asfort_athena_vehicle
                  (:True
                    (Class|VehicleFunctionLibrary|LaunchVehicle _asfort_athena_vehicle (CallFunction|CalculateImpulse _asfort_athena_vehicle (ScalableFloat|GetValueatLevel (Variables|ProtoCannon|GetHF_VehicleImpulseMultiplier)) (ScalableFloat|GetValueatLevel (Variables|ProtoCannon|GetHF_VehicleAdditionalVerticalImpulse))))))))))))))

(fn ApplyDamage
  (params (TargetActor :type Actor%20Object%20Reference))
  (bind _asfort_player_pawn (Utilities|Casting|CastToFortPlayerPawn TargetActor)
    (:then
      (if (== (Team|GetActorTeam _asfort_player_pawn) (Team|GetActorTeam (Variables|Actor|GetInstigator)))
        (bind _output_get (LocalVariables|SetDamageGEToApply "/PotatoCoaster/Gameplay/GAS/GE_PotatoCoaster_ClusterBomb_Damage_Mitigated.GE_PotatoCoaster_ClusterBomb_Damage_Mitigated_C"))
        (Fortnite|ApplyGameplayEffectsToActor (Variables|GetTargetActor) (Utilities|Array|MakeArray _output_get) (Transformation|GetActorLocation) false 0 (Ability|GetAbilitySystemComponent (Variables|Actor|GetInstigator)) (Class|AthenaFunctionLibrary|MakeSimpleFortEffectContext (Variables|Actor|GetInstigator) self))
        (elif (> (ScalableFloat|GetValueatLevel (Variables|Default|GetHF_UseRaritySpecificGE)) 0.5)
          (bind _asb_proto_cannon_cluster_bomb (Utilities|Casting|CastToB_ProtoCannon_ClusterBomb (Actor|GetOwner)))
          (switch Utilities|FlowControl|Switch|SwitchonEFortRarity (Rarity|GetWeaponDataRarity _asb_proto_cannon_cluster_bomb)
            (:Common
              _)
            (:Uncommon
              _)
            (:Rare
              _)
            (:Epic
              (LocalVariables|SetDamageGEToApply "/PotatoCoaster/Gameplay/GAS/GE_PotatoCoaster_ClusterBomb_Damage_Small_Epic.GE_PotatoCoaster_ClusterBomb_Damage_Small_Epic_C")
              (Fortnite|ApplyGameplayEffectsToActor (Variables|GetTargetActor) (Utilities|Array|MakeArray (LocalVariables|GetDamageGEToApply)) (Transformation|GetActorLocation) false 0 (Ability|GetAbilitySystemComponent (Variables|Actor|GetInstigator)) (Class|AthenaFunctionLibrary|MakeSimpleFortEffectContext (Variables|Actor|GetInstigator) self)))
            (:Legendary
              (LocalVariables|SetDamageGEToApply "/PotatoCoaster/Gameplay/GAS/GE_PotatoCoaster_ClusterBomb_Damage_Small_Legendary.GE_PotatoCoaster_ClusterBomb_Damage_Small_Legendary_C"))
            (:Mythic
              _)
            (:Transcendent
              _)
            (:Unattainable
              _))
          (LocalVariables|SetDamageGEToApply "/PotatoCoaster/Gameplay/GAS/GE_PotatoCoaster_ClusterBomb_Damage_Small.GE_PotatoCoaster_ClusterBomb_Damage_Small_C")
          (else
            (LocalVariables|SetDamageGEToApply "/PotatoCoaster/Gameplay/GAS/GE_PotatoCoaster_ClusterBomb_Damage_Small.GE_PotatoCoaster_ClusterBomb_Damage_Small_C")))))
    (:CastFailed
      _)))

(fn CalculateImpulse
  (params (ActorToImpulse :type Actor%20Object%20Reference) (ImpulseMultiplier :type Float%20%28double-precision%29) (AdditionalVerticalImpulse :type Float%20%28double-precision%29))
  (bind (_returnvalue_x _returnvalue_y _returnvalue_z) (Math|Vector|BreakVector (Transformation|GetActorLocation)))
  (if (> ImpulseMultiplier 0)
    (return (Math|Vector|MakeVector (.x (* (Math|Vector|Normalize (Math|Vector|vector-vector (Transformation|GetActorLocation ActorToImpulse) (Math|Vector|MakeVector _returnvalue_x _returnvalue_y (- _returnvalue_z (ScalableFloat|GetValueatLevel (Variables|ProtoCannon|GetHF_ZLocationOffset)))))) ImpulseMultiplier)) (.y (* (Math|Vector|Normalize (Math|Vector|vector-vector (Transformation|GetActorLocation ActorToImpulse) (Math|Vector|MakeVector _returnvalue_x _returnvalue_y (- _returnvalue_z (ScalableFloat|GetValueatLevel (Variables|ProtoCannon|GetHF_ZLocationOffset)))))) ImpulseMultiplier)) (+ (.z (* (Math|Vector|Normalize (Math|Vector|vector-vector (Transformation|GetActorLocation ActorToImpulse) (Math|Vector|MakeVector _returnvalue_x _returnvalue_y (- _returnvalue_z (ScalableFloat|GetValueatLevel (Variables|ProtoCannon|GetHF_ZLocationOffset)))))) ImpulseMultiplier)) AdditionalVerticalImpulse)))
    (else
      (return "0, 0, 0"))))

(fn ConstructionScript)
