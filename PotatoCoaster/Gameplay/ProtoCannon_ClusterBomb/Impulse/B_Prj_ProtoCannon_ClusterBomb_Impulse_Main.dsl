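; Main cluster bomb projectile for the ProtoCannon, which explodes and releases smaller grenades.
; A charge-scaled air-burst timer, stopping on impact, or entering water triggers one guarded explosion with FX and sound.
; The server runs a targeting preset to apply damage effects to player pawns, mitigated for teammates, and spawns timed grenades.
; Grenade directions fan around the hit normal with random variance, the first pointing along it.
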(event EventBeginPlay
  ;[0.03429 0.3069 0.7899] BeginPlay
  (|CallParentFunction)
  (Game|Components|ProjectileMovement|Interpolation|SetInterpolatedComponent (Variables|Movement|GetProjectileMovementComponent) (Variables|Default|GetProtoCannon_Ammo))
  (bind _returnvalue (Utilities|Time|SetTimerbyEvent (EventDispatchers|CreateEvent :!event Custom|AirBurstTrigger) (ScalableFloat|GetValueatLevel (Variables|Default|GetHF%20Air%20Burst%20Timer) (Math|Float|MapRangeClamped (Math|Float|Max(Float) (Variables|FortProjectileBase|GetChargePercent) 0.2) 0.2 1.0 0.0 1.0))))
  (Variables|Default|SetAirBurstTimerHandle _returnvalue))

(event Custom|AirBurstTrigger
  ;[0.03429 0.3069 0.7899] Air Burst Logic
  (if (not (Variables|Default|GetHasExploded))
    (Variables|Default|SetHasExploded true)
    (Components|Movement|StopMovementImmediately (Variables|Movement|GetProjectileMovementComponent))
    (Components|Activation|ToggleActive (Variables|Movement|GetProjectileMovementComponent))
    (Variables|Default|SetHitLocation (Transformation|GetActorLocation))
    (Variables|Default|SetHitNormal "0.000000,0.000000,-1.000000")
    (Effects|Components|ParticleSystem|SpawnFXAtLocation "/PotatoCoaster/Assets/FX/NS_PotatoCoaster_StandardExplosion.NS_PotatoCoaster_StandardExplosion" (Variables|Default|GetHitLocation))
    (Audio|PlaySoundatLocation "/PotatoCoaster/Assets/Sound/SC_ProtoCannon_Projectile_Land.SC_ProtoCannon_Projectile_Land" (Variables|Default|GetHitLocation))
    (CallFunction|DoonStop)
    (Rendering|SetVisibility (Variables|Default|GetProtoCannon_Ammo) false false)
    (CallFunction|DoonExplode)))

(event Custom|DoOnStop
  (Networking|SwitchHasAuthority
    (:Authority
      (bind (_asynctaskproxy _targetinghandle) (Class|AsyncActionPerformTargeting|PerformTargetingAsyncAction 0 "/PotatoCoaster/Gameplay/ProtoCannon_ClusterBomb/Impulse/TP_ProtoCannon_Explosion.TP_ProtoCannon_Explosion" true)
        (:Targeted
          (for _array_element (TargetingSystem|TargetingResults|GetTargetingResults (GameInstanceSubsystems|GetTargetingSubsystem) _targetinghandle)
            (bind (_array_element_bblockinghit _array_element_binitialoverlap _array_element_time _array_element_distance _array_element_location _array_element_impactpoint _array_element_normal _array_element_impactnormal _array_element_physmat _array_element_hitactor) (Collision|BreakHitResult _array_element))
            (CallFunction|ApplyDamage _array_element_hitactor)))))))

(event EventOnStop
  (params (Hit :type Hit%20Result%20Structure%20%28by%20ref%29))
  (bind (_bblockinghit _binitialoverlap _time _distance _location _impactpoint _normal _impactnormal _physmat _hitactor) (Collision|BreakHitResult Hit))
  ;[0.03429 0.3069 0.7899] Projectile stopped moving
  (if (not (Variables|Default|GetHasExploded))
    (Variables|Default|SetHasExploded true)
    (Utilities|Time|ClearandInvalidateTimerbyHandle (Variables|Default|GetAirBurstTimerHandle))
    (CallFunction|DoonStop)
    (Variables|Default|SetHitLocation _location)
    (if (< (.z _impactnormal) 0)
      (Variables|Default|SetHitNormal "0.000000,0.000000,-1.000000")
      (else
        (Variables|Default|SetHitNormal "0.000000,0.000000,1.000000")))
    (Effects|Components|ParticleSystem|SpawnFXAtLocation "/PotatoCoaster/Assets/FX/NS_PotatoCoaster_StandardExplosion.NS_PotatoCoaster_StandardExplosion" (Variables|Default|GetHitLocation))
    (Audio|PlaySoundatLocation "/PotatoCoaster/Assets/Sound/SC_ProtoCannon_Projectile_Land.SC_ProtoCannon_Projectile_Land" (Variables|Default|GetHitLocation))
    (Rendering|SetVisibility (Variables|Default|GetProtoCannon_Ammo) false false)
    (Variables|Default|SetbHitMovingPlatform (Utilities|IsValid (Actor|GetComponentbyClass _hitactor "/Script/FortniteGame.FortMovingPlatformComponent") :!class /Script/Engine.KismetSystemLibrary))
    (bind _returnvalue (GameplayTags|HasAnyMatchingGameplayTags(Message) _hitactor (Variables|Default|GetAttachableTags)))
    (if _returnvalue
      (Transformation|AttachActorToComponent (Variables|/Script/Engine.Actor|GetRootComponent _hitactor) "None" "KeepWorld" "KeepWorld" "KeepWorld" false)
      (else
        _))
    (Utilities|FlowControl|Delay (ScalableFloat|GetValueatLevel (Variables|Default|GetHF_ClusterSpawnDelay)))
    (CallFunction|DoonExplode)))

(event OnEnterWaterEvent(WaterInteractionComponent)
  (params (WaterBody :type Fort%20Water%20Body%20Actor%20Object%20Reference) (WaterInteractionComponent :type Fort%20Water%20Interaction%20Component%20Object%20Reference) (bIsFirstBody :type Boolean))
  (bind (_waterplanelocation _waterplanenormal _watersurfaceposition _waterdepth _waterbodyidx) (FX|GetLastWaterSurfaceInfo WaterInteractionComponent :!class /Script/FortniteGame.FortWaterInteractionComponent))
  ;[0.03429 0.3069 0.7899] Handle Entering Water
  (Utilities|FlowControl|DoOnce false
    (:Completed
      (if (not (Variables|Default|GetHasExploded))
        (Variables|Default|SetHasExploded true)
        (Utilities|Time|ClearandInvalidateTimerbyHandle (Variables|Default|GetAirBurstTimerHandle))
        (CallFunction|DoonStop)
        (Variables|Default|SetHitLocation (+ _watersurfaceposition "0.000000,0.000000,50.000000"))
        (Variables|Default|SetHitNormal "0, 0,1.000000")
        (bind _returnvalue (Effects|SpawnEmitteratLocation "/Game/Weapons/FORT_Generic_Effects/Impacts/Water/P_Pistol_Impact_Water_Apollo.P_Pistol_Impact_Water_Apollo" _watersurfaceposition (Math|Rotator|MakeRotfromX _waterplanenormal)))
        (Audio|PlaySoundatLocation "/ClusterBombGameplay/Audio/ClusterBomb_ThrowLand_Water.ClusterBomb_ThrowLand_Water" _watersurfaceposition)
        (Effects|Components|ParticleSystem|SetFloatParameter _returnvalue "WaterDepth" _waterdepth)
        (Effects|Components|ParticleSystem|SetFloatParameter _returnvalue "WaterBodyIndex" (Math|Conversions|ToFloat(Integer) _waterbodyidx))
        (Audio|PlaySoundatLocation "/Game/Sounds/Fort_GamePlay_Sounds/Explosives/Grenade_Projectile_Bounce_Water_Splash_NoLoop_Cue.Grenade_Projectile_Bounce_Water_Splash_NoLoop_Cue" (Transformation|GetActorLocation))
        (Components|Movement|StopMovementImmediately (Variables|Movement|GetProjectileMovementComponent))
        (Components|Activation|ToggleActive (Variables|Movement|GetProjectileMovementComponent))
        (Effects|Components|ParticleSystem|SpawnFXAtLocation "/Game/Effects/Explosion_Update/NS_GrenadeLauncher_Explosion_Asteria_Water.NS_GrenadeLauncher_Explosion_Asteria_Water" (Variables|Default|GetHitLocation))
        (Audio|PlaySoundatLocation "/PotatoCoaster/Assets/Sound/SC_ProtoCannon_Projectile_Land.SC_ProtoCannon_Projectile_Land" (Variables|Default|GetHitLocation))
        (Rendering|SetVisibility (Variables|Default|GetProtoCannon_Ammo) false false)
        (Utilities|FlowControl|Delay (ScalableFloat|GetValueatLevel (Variables|Default|GetHF_ClusterSpawnDelay)))
        (CallFunction|DoonExplode)))))

(event Custom|DoOnExplode
  ;[0.03429 0.3069 0.7899] Explode and spawn small grenades
  (Collision|SetActorEnableCollisionwithChildren 0)
  (Networking|SwitchHasAuthority
    (:Authority
      (bind _returnvalue (Utilities|Time|SetTimerbyEvent (delegate Custom|SpawnGrenadeTimed) (ScalableFloat|GetValueatLevel (Variables|Default|GetHF_GrenadeInterval)) true))
      (Variables|Default|SetGrenadeSpawnTimer _returnvalue)
      (if (> (ScalableFloat|GetValueasInteger (Variables|Default|GetHF_GrenadesToSpawn)) (Variables|Default|GetGrenadesSpawned))
        (CallFunction|SpawnGrenadeSmall 0)
        (else
          (Utilities|Time|ClearandInvalidateTimerbyHandle (Variables|Default|GetGrenadeSpawnTimer)))))
    (:Remote
      _))
  (Utilities|FlowControl|Delay 0.2)
  (Actor|DestroyActor))

(event Custom|SpawnGrenadeTimed
  ;[0.03429 0.3069 0.7899] Explode and spawn small grenades
  (if (> (ScalableFloat|GetValueasInteger (Variables|Default|GetHF_GrenadesToSpawn)) (Variables|Default|GetGrenadesSpawned))
    (CallFunction|SpawnGrenadeSmall (Variables|Default|GetGrenadesSpawned))
    (if (== (ScalableFloat|GetValueasInteger (Variables|Default|GetHF_GrenadesToSpawn)) (Variables|Default|GetGrenadesSpawned))
      _)
    (else
      _))
  (Utilities|Time|ClearandInvalidateTimerbyHandle (Variables|Default|GetGrenadeSpawnTimer))
  (Utilities|FlowControl|Delay 0.2)
  (Actor|DestroyActor))

(fn ApplyDamage
  (params (TargetActor :type Actor%20Object%20Reference))
  (bind _asfort_player_pawn (Utilities|Casting|CastToFortPlayerPawn TargetActor)
    (:then
      (if (== (Team|GetActorTeam _asfort_player_pawn) (Team|GetActorTeam (Variables|Actor|GetInstigator)))
        (bind _output_get (LocalVariables|SetDamageGEToApply "/PotatoCoaster/Gameplay/GAS/GE_PotatoCoaster_ClusterBomb_Damage_Mitigated.GE_PotatoCoaster_ClusterBomb_Damage_Mitigated_C"))
        (Fortnite|ApplyGameplayEffectsToActor (Variables|GetTargetActor) (Utilities|Array|MakeArray _output_get) (Transformation|GetActorLocation) false 0 (Ability|GetAbilitySystemComponent (Variables|Actor|GetInstigator)) (Class|AthenaFunctionLibrary|MakeSimpleFortEffectContext (Variables|Actor|GetInstigator) self))
        (elif (> (ScalableFloat|GetValueatLevel (Variables|Default|GetHF_UseRaritySpecificGE)) 0.5)
          (bind _asb_proto_cannon_cluster_bomb (Utilities|Casting|CastToB_ProtoCannon_ClusterBomb (Actor|GetOwner)))
          (switch Utilities|FlowControl|Switch|SwitchonEFortRarity (Rarity|GetWeaponDataRarity _asb_proto_cannon_cluster_bomb)
            (:Common
              _)
            (:Uncommon
              _)
            (:Rare
              _)
            (:Epic
              (LocalVariables|SetDamageGEToApply "/PotatoCoaster/Gameplay/GAS/GE_PotatoCoaster_ClusterBomb_Damage_Main_Epic.GE_PotatoCoaster_ClusterBomb_Damage_Main_Epic_C")
              (Fortnite|ApplyGameplayEffectsToActor (Variables|GetTargetActor) (Utilities|Array|MakeArray (LocalVariables|GetDamageGEToApply)) (Transformation|GetActorLocation) false 0 (Ability|GetAbilitySystemComponent (Variables|Actor|GetInstigator)) (Class|AthenaFunctionLibrary|MakeSimpleFortEffectContext (Variables|Actor|GetInstigator) self)))
            (:Legendary
              (LocalVariables|SetDamageGEToApply "/PotatoCoaster/Gameplay/GAS/GE_PotatoCoaster_ClusterBomb_Damage_Main_Legendary.GE_PotatoCoaster_ClusterBomb_Damage_Main_Legendary_C"))
            (:Mythic
              _)
            (:Transcendent
              _)
            (:Unattainable
              _))
          (LocalVariables|SetDamageGEToApply "/PotatoCoaster/Gameplay/GAS/GE_PotatoCoaster_ClusterBomb_Damage_Main.GE_PotatoCoaster_ClusterBomb_Damage_Main_C")
          (else
            (LocalVariables|SetDamageGEToApply "/PotatoCoaster/Gameplay/GAS/GE_PotatoCoaster_ClusterBomb_Damage_Main.GE_PotatoCoaster_ClusterBomb_Damage_Main_C")))))
    (:CastFailed
      _)))

(fn CalculateSmallGrenadeSpawnDirection
  (params (Index :type Integer) (HitNormal :type Vector))
  (bind _index (Variables|GetIndex))
  (bind _returnvalue (== _index 0))
  (bind _grenadesmallspawnverticalangle (Variables|SmallGrenade|GetGrenadeSmallSpawnVerticalAngle))
  (bind _returnvalue_2 (Math|Float|SelectFloat 9e+01 _grenadesmallspawnverticalangle _returnvalue))
  (bind _hitnormal (Variables|Default|GetHitNormal))
  (bind _crossproduct (LocalVariables|GetCrossProduct))
  (bind _returnvalue_3 (Math|Vector|CrossProduct _crossproduct _hitnormal))
  (bind _crossproduct_2 (LocalVariables|GetCrossProduct))
  (bind _returnvalue_4 (Math|Vector|RotateVectorAroundAxis _crossproduct_2 _returnvalue_2 _returnvalue_3))
  (bind _hf_grenadestospawn (Variables|Default|GetHF_GrenadesToSpawn))
  (bind _returnvalue_5 (ScalableFloat|GetValueasInteger _hf_grenadestospawn))
  (bind _grenadesmallhorizontalanglevariance (Variables|SmallGrenade|GetGrenadeSmallHorizontalAngleVariance))
  (bind _grenadesmallhorizontalanglevariance_2 (Variables|SmallGrenade|GetGrenadeSmallHorizontalAngleVariance))
  (bind _returnvalue_6 (* _grenadesmallhorizontalanglevariance_2 -1.0))
  (bind _returnvalue_7 (Math|Random|RandomFloatinRange _returnvalue_6 _grenadesmallhorizontalanglevariance))
  (bind _returnvalue_8 (- _returnvalue_5 1))
  (bind _returnvalue_9 (/ 360 _returnvalue_8))
  (bind _index_2 (Variables|GetIndex))
  (bind _returnvalue_10 (* _index_2 _returnvalue_9))
  (bind _returnvalue_11 (Math|Conversions|ToFloat(Integer) _returnvalue_10))
  (bind _hitnormal_2 (Variables|Default|GetHitNormal))
  (bind _returnvalue_12 (+ _returnvalue_11 _returnvalue_7))
  (bind _returnvalue_13 (Math|Vector|RotateVectorAroundAxis _returnvalue_4 _returnvalue_12 _hitnormal_2))
  ;[1 1 1] Cross product to get vector parallel with surface
  (bind _output_get (LocalVariables|SetCrossProduct (Math|Vector|CrossProduct HitNormal (select (> (Math|Float|Absolute(Float) (.x HitNormal)) 0.9) (LocalVariables|GetYAxis) (LocalVariables|GetXAxis)))))
  (Debug|DrawDebugDirectionalArrow _output_get "(R=1.000000,G=0.000000,B=0.000000,A=1.000000)")
  ;[1 1 1] Rotate around HitNormal based on Index and horizontal angle variance
  (Debug|DrawDebugDirectionalArrow _returnvalue_13 "(R=0.933333,G=0.000000,B=1.000000,A=1.000000)")
  (return (Math|Conversions|RotationFromXVector _returnvalue_13)))

(fn DrawDebugDirectionalArrow
  (params (LineEnd :type Vector) (LineColor :type Linear%20Color%20Structure))
  (if (and (Variables|Debug|GetbDrawDebugDirectionalArrows) (Editor|InEditororPIE))
    (Rendering|Debug|DrawDebugArrow (Transformation|GetActorLocation) (+ (Transformation|GetActorLocation) (* (Variables|GetLineEnd) 5e+01)) 1e+01 (Variables|GetLineColor) 5.0 2.0)))

(fn SpawnGrenadeSmall
  (params (Index :type Integer))
  (LocalVariables|SetExplosionDelay (* Index (ScalableFloat|GetValueatLevel (Variables|Default|GetHF_GrenadeInterval))))
  (bind _returnvalue (Ability|MakeGameplayEffectContainerSpecfromAbilityComponent (Variables|Default|GetGrenadeEffectContainer) (Ability|GetFortAbilitySystemComponent (Game|GetInstigator))))
  (Debug|DrawDebugDirectionalArrow (Variables|Default|GetHitNormal) "(R=1.000000,G=0.000000,B=0.000000,A=1.000000)")
  (bind _spawnrotation (CallFunction|CalculateSmallGrenadeSpawnDirection (Variables|GetIndex) (Variables|Default|GetHitNormal)))
  (bind _returnvalue_2 (Projectile|SpawnProjectilewithParams (Variables|Default|GetGrenadeSmallClass) (Game|GetInstigator) (Utilities|Struct|MakeSpawnProjectileParams (Variables|Default|GetHitLocation) "0.000000,0.000000,0.000000" _spawnrotation "(TargetSelection=(List=,bStopAtFirstSuccess=True,bKeepCheckingListOnIndestructibleHit=False,AbilityTargetBlockedGameplayTags=(GameplayTags=),bUseWeaponRanges=True,bUseMaxYawAngleToTarget=False,MaxYawAngleToTarget=0.000000),TargetGameplayEffectSpecs=,OwnerGameplayEffectSpecs=,ActivationCues=(GameplayTags=),ImpactCues=(GameplayTags=),ImpactNoiseRange=0.000000,FlyByNoiseRange=0.000000,bOverrideChargeMagnitude=False,ChargeMagnitudeOverrideValue=0.000000)" _returnvalue false 0 (ScalableFloat|GetValueatLevel (Variables|SmallGrenade|GetHF_SmallGrenadeInitialSpeed)) (ScalableFloat|GetValueatLevel (Variables|SmallGrenade|GetHF_SmallGrenadeInitialSpeed)) (ScalableFloat|GetValueatLevel (Variables|SmallGrenade|GetHF_SmallGrenadeGravity)))))
  (Actor|SetOwner _returnvalue_2 (Actor|GetOwner))
  (bind _asb_prj_paprika_cluster_bomb_grenade_small (Utilities|Casting|CastToB_Prj_Paprika_ClusterBomb_GrenadeSmall _returnvalue_2)
    (:then
      (Variables|/ClusterBombGameplay/Gameplay/Actors/B_Prj_Paprika_ClusterBomb_GrenadeSmall.B_Prj_Paprika_ClusterBomb_GrenadeSmall_C|SetbMainGrenadeStuckToMovingPlatform (Variables|Default|GetbHitMovingPlatform) _asb_prj_paprika_cluster_bomb_grenade_small))
    (:CastFailed
      _))
  (Math|Integer|IncrementInt (Variables|Default|GetGrenadesSpawned)
    (:%20%20
      (return))))

(fn ConstructionScript)
