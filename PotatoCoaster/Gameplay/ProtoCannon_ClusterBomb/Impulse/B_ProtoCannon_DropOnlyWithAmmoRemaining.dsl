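; Drop-item custom behavior for the ProtoCannon ClusterBomb, named to drop the item only while it has ammo remaining.
; HandleItemSpawn breaks the drop parameters into the item entry and reads its phantom reserve ammo.
; It returns true when that reserve is above zero and false otherwise, without checking loaded ammo or the pawn.
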(fn HandleItemSpawn
  (params (AvatarPawn :type Fort%20Pawn%20Object%20Reference) (Params :type Drop%20Item%20Custom%20Behavior%20Parameters%20Structure%20%28by%20ref%29))
  (bind (_itemdefinition _itemguid _count _level _loadedammo _phantomreserveammo) (Class|FortItemFunctionLibrary|BreakFortItemEntry (Utilities|Struct|BreakDropItemCustomBehaviorParameters Params)))
  (if (> _phantomreserveammo 0)
    (return true)
    (else
      (return false))))
